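# Translucent unlit material for Niagara ribbons that draws the chroma tail effect in the player's Mouse emote.
# A mask texture, tiled twice along U, tints its red, green and blue channels cyan, blue and purple, blended by successive lerps on green and blue.
# A second chain multiplies particle alpha by a channel sum that counts red twice and omits green.
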Material /Game/Animation/Game/MainPlayer/Emotes/Mouse/FX/Material/M_Mouse_ChromaTail
graph {
 properties {bUsedWithNiagaraRibbons=true, blendMode=BLEND_Translucent, pixelDepthOffsetMode=PDOM_Legacy, shadingModel=MSM_Unlit, usedShadingModels=Unlit}
 nodes [
  n1 = Constant3Vector() {constant=<rgba 0 0.6393 1 0>}
  n2 = TextureSample(UVs=TextureCoordinate() {uTiling=2}) {samplerType=SAMPLERTYPE_LinearColor, texture=/Game/Animation/Game/MainPlayer/Emotes/Mouse/FX/Textures/T_MouseChromaTail_M}
  n3 = Constant3Vector() {constant=<rgba 0 0.1344 1 0>}
  LinearInterpolate(
   A=LinearInterpolate(A=Multiply(A=@n1, B=@n2.R), B=Multiply(A=@n3, B=@n2.G), Alpha=@n2.G)
   B=Multiply(A=Constant3Vector() {constant=<rgba 0.7686 0 1 0>}, B=@n2.B)
   Alpha=@n2.B
  )
  Multiply(A=ParticleColor().A, B=Add(A=Add(A=@n2.R, B=@n2.R), B=@n2.B))
 ]
}
